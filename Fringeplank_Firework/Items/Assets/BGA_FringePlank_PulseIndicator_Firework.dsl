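; Fringe Plank pulse indicator firework actor that plays a one-shot spark explosion effect and sound at its own location.
; BeginPlay calls PlayFX, which spawns a pooled, auto-release Niagara system and plays the explosion sound.
; The firework color is passed to the spawned effect as a float parameter, with the parameter name and value taken from variables.
; Overlap and Tick events are disabled stubs.
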(event EventBeginPlay
  (|CallParentFunction)
  (CallFunction|PlayFX))

(event Collision|EventActorBeginOverlap
  (params (OtherActor :type Actor%20Object%20Reference))
  (|CallParentFunction OtherActor) ; This node is disabled and will not be called.\nDrag off pins to build functionality.
)

(event EventTick
  (params (DeltaSeconds :type Float%20%28single-precision%29))
  (|CallParentFunction DeltaSeconds) ; This node is disabled and will not be called.\nDrag off pins to build functionality.
)

(event Custom|PlayFX
  (bind _returnvalue (Niagara|SpawnSystematLocation :SystemTemplate (Variables|Default|GetSparkExplosion_NS) :Location (Transformation|GetActorLocation) :Scale "1.000000,1.000000,1.000000" :bAutoDestroy true :bAutoActivate true :PoolingMethod "AutoRelease" :bPreCullCheck false))
  (Effects|Components|ParticleSystem|SetFloatParameter _returnvalue (Variables|Default|GetFirework%20Color%20Parameter) (Variables|Default|GetFireworkColor))
  (Audio|PlaySoundatLocation (Variables|Default|GetSparkExplosion_Sound) (Transformation|GetActorLocation)))

(fn ConstructionScript
  (|CallParentFunction))
